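# Physics asset for the F_MED_TangyRadishMagma_FE female body cosmetic, set up for secondary motion on a hanging belt piece.
# A two-link chain runs from base_belt_02 through a capsule on dyn_belt_02 to a box on dyn_belt_03, with tight angular limits at each link.
# Capsules on the pelvis and both thigh movers provide body shapes around the belt.
# The belt base and thighs join the pelvis with free angular constraints.
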PhysicsAsset /BRCosmetics/Characters/Player/Female/Medium/Bodies/F_MED_TangyRadishMagma/Meshes/F_MED_TangyRadishMagma_FE_Physics
bodies {
 base_belt_02 [{shapeType=Capsule, length=0.505, radius=0.505, shapeName=base_belt_02_capsule, rotation=<rot 0 -90 0>}]
 dyn_belt_02 [{
   shapeType=Capsule
   length=2.9221
   radius=1.0105
   shapeName=dyn_belt_02_capsule
   center=<2 0.0944 -0.2795>
   rotation=<rot 88.8271 -0.0008 -0.0008>
  }]
 dyn_belt_03 [{
   shapeType=Box
   extentX=12.2957
   extentY=8.5544
   extentZ=1.5069
   shapeName=dyn_belt_03_box
   center=<5.4878 0 -0.4739>
   rotation=<rot -3.3283 -0.0001 0>
  }]
 pelvis [{shapeType=Capsule, length=10.0572, radius=9.2636, shapeName=pelvis_capsule, center=<-4.5141 -0.6624 0.3745>}]
 thigh_l_mover [{
   shapeType=Capsule
   length=14.0989
   radius=9.7864
   shapeName=thigh_l_capsule
   center=<-10.8446 -0.3453 1.6145>
   rotation=<rot 82.6177 176.0029 -177.7185>
  }]
 thigh_r_mover [{
   shapeType=Capsule
   length=13.9945
   radius=9.8907
   shapeName=thigh_l_capsule
   center=<10.7289 -0.0382 -1.0442>
   rotation=<rot 82.6177 176.0036 -174.6015>
  }]
}
constraints [
 {bone1Name=dyn_belt_02, bone2Name=base_belt_02, swing1LimitDegrees=1, swing2LimitDegrees=25, twistLimitDegrees=10}
 {bone1Name=dyn_belt_03, bone2Name=dyn_belt_02, swing1LimitDegrees=10, swing2LimitDegrees=25, twistLimitDegrees=1}
 {
  bone1Name=base_belt_02
  bone2Name=pelvis
  swing1LimitDegrees=45
  swing1Motion=Free
  swing2LimitDegrees=45
  swing2Motion=Free
  twistLimitDegrees=45
  twistMotion=Free
 }
 {
  bone1Name=thigh_l_mover
  bone2Name=pelvis
  swing1LimitDegrees=45
  swing1Motion=Free
  swing2LimitDegrees=45
  swing2Motion=Free
  twistLimitDegrees=45
  twistMotion=Free
 }
 {
  bone1Name=thigh_r_mover
  bone2Name=pelvis
  swing1LimitDegrees=45
  swing1Motion=Free
  swing2LimitDegrees=45
  swing2Motion=Free
  twistLimitDegrees=45
  twistMotion=Free
 }
]
