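# Physics asset for the TangyRadishMagma medium female body cosmetic, covering a short dangling belt chain and the hips around it.
# Small capsules on base_belt_02, dyn_belt_02 and dyn_belt_03 form a chain with narrow angular limits that favor swing on one axis.
# Larger pelvis and thigh-mover capsules approximate the body volume near the belt.
# The belt base and both thighs attach to the pelvis through constraints with free rotation.
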PhysicsAsset /BRCosmetics/Characters/Player/Female/Medium/Bodies/F_MED_TangyRadishMagma/Meshes/F_MED_TangyRadishMagma_Physics
bodies {
 base_belt_02 [{shapeType=Capsule, length=0.505, radius=0.505, shapeName=base_belt_02_capsule, rotation=<rot 0 -90 0>}]
 dyn_belt_02 [{
   shapeType=Capsule
   length=2.9221
   radius=1.0105
   shapeName=dyn_belt_02_capsule
   center=<2 0.0944 -0.2795>
   rotation=<rot 88.8271 -0.0008 -0.0008>
  }]
 dyn_belt_03 [{
   shapeType=Capsule
   length=2.8925
   radius=1.4462
   shapeName=dyn_belt_03_capsule
   center=<3.7872 0 -0.7317>
   rotation=<rot 84.1423 -0.0007 -0.0005>
  }]
 pelvis [{shapeType=Capsule, length=10.0572, radius=9.2636, shapeName=pelvis_capsule, center=<-4.5141 0.0057 0.3745>}]
 thigh_l_mover [{
   shapeType=Capsule
   length=13.9945
   radius=9.8907
   shapeName=thigh_l_capsule
   center=<-10.8827 -0.6239 1.6121>
   rotation=<rot 82.6177 176.0031 -176.1953>
  }]
 thigh_r_mover [{
   shapeType=Capsule
   length=13.9945
   radius=9.8907
   shapeName=thigh_l_capsule
   center=<10.7289 -0.0382 -1.0442>
   rotation=<rot 82.6177 176.0036 -174.6015>
  }]
}
constraints [
 {bone1Name=dyn_belt_02, bone2Name=base_belt_02, swing1LimitDegrees=1, swing2LimitDegrees=25, twistLimitDegrees=1}
 {bone1Name=dyn_belt_03, bone2Name=dyn_belt_02, swing1LimitDegrees=10, swing2LimitDegrees=25, twistLimitDegrees=1}
 {
  bone1Name=base_belt_02
  bone2Name=pelvis
  swing1LimitDegrees=45
  swing1Motion=Free
  swing2LimitDegrees=45
  swing2Motion=Free
  twistLimitDegrees=45
  twistMotion=Free
 }
 {
  bone1Name=thigh_l_mover
  bone2Name=pelvis
  swing1LimitDegrees=45
  swing1Motion=Free
  swing2LimitDegrees=45
  swing2Motion=Free
  twistLimitDegrees=45
  twistMotion=Free
 }
 {
  bone1Name=thigh_r_mover
  bone2Name=pelvis
  swing1LimitDegrees=45
  swing1Motion=Free
  swing2LimitDegrees=45
  swing2Motion=Free
  twistLimitDegrees=45
  twistMotion=Free
 }
]
