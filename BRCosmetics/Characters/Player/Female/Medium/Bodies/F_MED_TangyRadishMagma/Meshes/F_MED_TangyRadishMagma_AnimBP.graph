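# Post-process style AnimBlueprint for the F_MED_TangyRadishMagma player body cosmetic, copying the parent mesh pose and adding secondary motion and correctives.
# A Control Rig, fed crouch, movement, melee and front-end flags, leads into bone offsets, a bool blend of two RigidBody sims, then AnimDynamics skirt springs.
# The front-end `Frontend_Idle_Bypass` curve weights a DefaultSlot arm layer before final forearm, elbow-strap and wrist corrections.
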AnimBlueprint /BRCosmetics/Characters/Player/Female/Medium/Bodies/F_MED_TangyRadishMagma/Meshes/F_MED_TangyRadishMagma_AnimBP
target_skeleton=&/F_MED_TangyRadishMagma_Skeleton
graphs [
 AnimGraph {nodes [
   Anim Dynamics - Bone: dyn_backSkirt_02_l(Alpha=1, ExternalForce="0,0,0") {
    angularDampingOverride=0.999
    angularSpringConstant=1000
    bAngularSpring=true
    bOverrideAngularDamping=true
    bUseGravityOverride=true
    gravityScale=0.35
    simulationSpace=RootRelative
    alphaBoolBlend {blendOption=Linear}
    boundBone {boneName=dyn_backSkirt_02_l}
    componentLinearAccScale=<1 1 1>
    componentLinearVelScale=<1 1 1>
    gravityOverride {z=-980}
    physicsBodyDefinitions [{
      boundBone {boneName=dyn_backSkirt_02_l}
      boxExtents=<15 15 15>
      constraintSetup {angularLimitsMax=<0 0 0>, angularLimitsMin=<0 -50 0>, angularTarget=<1 0 0>}
      localJointOffset=<2 0 0>
     }]
    relativeSpaceBone {boneName=spine_01}
   }
   n1 = Layered blend per bone(
    BasePose=Use cached pose 'animDynamics Skirt'()
    BlendPoses_0=Slot 'DefaultSlot'()
    BlendWeights_0=Get Custom Idle is Playing Float() -> (Custom Idle is Playing Float:double)
   ) {layerSetup [{branchFilters [
       {boneName=upperarm_r}
       {boneName=upperarm_twist_01_r}
       {boneName=upperarm_twist_02_r}
       {boneName=lowerarm_r}
       {boneName=lowerarm_twist_01_r}
       {boneName=lowerarm_twist_02_r}
       {boneName=hand_r}
       {boneName=upperarm_l}
       {boneName=upperarm_twist_01_l}
       {boneName=upperarm_twist_02_l}
       {boneName=lowerarm_l}
       {boneName=lowerarm_twist_01_l}
       {boneName=lowerarm_twist_02_l}
       {boneName=hand_l}
      ]}]}
   n2 = "Transform (Modify) Bone - Bone: lowerarm_twist_01_r"(ComponentPose=Local To Component(LocalPose=@n1), Alpha=1, Translation="-0.454,-1.095,-0.250", Rotation="3.5,0,0", Scale="1,1.1,1.1") {
    rotationMode=BMM_Additive
    rotationSpace=BCS_BoneSpace
    scaleMode=BMM_Additive
    scaleSpace=BCS_BoneSpace
    translationMode=BMM_Additive
    translationSpace=BCS_BoneSpace
    alphaBoolBlend {blendOption=Linear}
    boneToModify {boneName=lowerarm_twist_01_r}
    rotation {pitch=3.5}
    scale {y=1.1, z=1.1}
    translation=<-0.454 -1.095 -0.25>
   }
   n3 = "Transform (Modify) Bone - Bone: lowerarm_twist_01_l"(ComponentPose=@n2, Alpha=1, Translation="0,0.15,0.5", Rotation="4,0,0", Scale="1,1.1,1.1") {
    rotationMode=BMM_Additive
    rotationSpace=BCS_BoneSpace
    scaleMode=BMM_Additive
    scaleSpace=BCS_BoneSpace
    translationMode=BMM_Additive
    translationSpace=BCS_BoneSpace
    alphaBoolBlend {blendOption=Linear}
    boneToModify {boneName=lowerarm_twist_01_l}
    rotation {pitch=4}
    scale {y=1.1, z=1.1}
    translation {y=0.15, z=0.5}
   }
   n4 = "Transform (Modify) Bone - Bone: elbowStrap_Corrective"(ComponentPose=@n3, Alpha=1, Translation="3.5,0,0", Rotation="0,0,0", Scale="1,1,1") {
    rotationMode=BMM_Additive
    rotationSpace=BCS_BoneSpace
    scaleMode=BMM_Additive
    scaleSpace=BCS_BoneSpace
    translationMode=BMM_Additive
    translationSpace=BCS_BoneSpace
    alphaBoolBlend {blendOption=Linear}
    boneToModify {boneName=elbowStrap_Corrective}
    translation {x=3.5}
   }
   n5 = "Transform (Modify) Bone - Bone: deform_wrist_dn_r"(ComponentPose=@n4, Alpha=1, Translation="0.25,3,-0.5", Rotation="0,-30,0", Scale="1,1,1") {
    rotationMode=BMM_Additive
    rotationSpace=BCS_BoneSpace
    scaleMode=BMM_Additive
    scaleSpace=BCS_BoneSpace
    translationMode=BMM_Additive
    translationSpace=BCS_BoneSpace
    alphaBoolBlend {blendOption=Linear}
    boneToModify {boneName=deform_wrist_dn_r}
    rotation {yaw=-30}
    translation=<0.25 3 -0.5>
   }
   n6 = "Transform (Modify) Bone - Bone: deform_wrist_up_r"(ComponentPose=@n5, Alpha=1, Translation="0,-2,0", Rotation="0,30,0", Scale="0.2,1,1") {
    rotationMode=BMM_Additive
    rotationSpace=BCS_BoneSpace
    scaleMode=BMM_Additive
    scaleSpace=BCS_BoneSpace
    translationMode=BMM_Additive
    translationSpace=BCS_BoneSpace
    alphaBoolBlend {blendOption=Linear}
    boneToModify {boneName=deform_wrist_up_r}
    rotation {yaw=30}
    scale {x=0.2}
    translation {y=-2}
   }
   n7 = "Transform (Modify) Bone - Bone: deform_wrist_dn_l"(ComponentPose=@n6, Alpha=1, Translation="0,-1,0", Rotation="0,-25,0", Scale="0.2,1,1") {
    rotationMode=BMM_Additive
    rotationSpace=BCS_BoneSpace
    scaleMode=BMM_Additive
    scaleSpace=BCS_BoneSpace
    translationMode=BMM_Additive
    translationSpace=BCS_BoneSpace
    alphaBoolBlend {blendOption=Linear}
    boneToModify {boneName=deform_wrist_dn_l}
    rotation {yaw=-25}
    scale {x=0.2}
    translation {y=-1}
   }
   n8 = "Transform (Modify) Bone - Bone: deform_wrist_up_l"(ComponentPose=@n7, Alpha=1, Translation="0,3,0", Rotation="0,0,0", Scale="0.2,1,1") {
    rotationMode=BMM_Additive
    rotationSpace=BCS_BoneSpace
    scaleMode=BMM_Additive
    scaleSpace=BCS_BoneSpace
    translationMode=BMM_Additive
    translationSpace=BCS_BoneSpace
    alphaBoolBlend {blendOption=Linear}
    boneToModify {boneName=deform_wrist_up_l}
    scale {x=0.2}
    translation {y=3}
   }
   Output Pose(Result=Component To Local(ComponentPose=@n8)) {layerGroup=None}
   n9 = RigidBody!(ComponentPose=Local To Component(LocalPose=Use cached pose 'TransformCache'()), Alpha, OverrideWorldGravity:Vector="0,0,-980", ExternalForce) -> (Pose) {
    bClampLinearTranslationLimitToRefPose=true
    bOverrideWorldGravity=true
    overridePhysicsAsset=&/F_MED_TangyRadishMagma_Physics
    simulationSpace=BaseBoneSpace
    alphaBoolBlend {blendOption=Linear}
    baseBoneRef {boneName=pelvis}
    simSpaceSettings {
     dampingAlpha=0
     maxAngularAcceleration=32
     maxAngularVelocity=8
     maxLinearAcceleration=2800
     maxLinearVelocity=700
     velocityScaleZ=0.4
     worldAlpha=1
     externalLinearDragV=<0.5 0.5 0.5>
    }
   }
   n10 = Component To Local(ComponentPose=@n9)
   n11 = RigidBody(ComponentPose=Local To Component(LocalPose=Use cached pose 'TransformCache'()), OverrideWorldGravity="0,0,-980") {
    bClampLinearTranslationLimitToRefPose=true
    bOverrideWorldGravity=true
    overridePhysicsAsset=&/F_MED_TangyRadishMagma_FE_Physics
    simulationSpace=BaseBoneSpace
    alphaBoolBlend {blendOption=Linear}
    baseBoneRef {boneName=pelvis}
    simSpaceSettings {
     dampingAlpha=0
     maxAngularAcceleration=32
     maxAngularVelocity=8
     maxLinearAcceleration=2800
     maxLinearVelocity=700
     worldAlpha=1
     externalLinearDragV=<0.5 0.5 0.5>
    }
   }
   n12 = Component To Local(ComponentPose=@n11)
   n13 = Anim Dynamics - Bone: dyn_frontSkirt_02_r(
    ComponentPose=Local To Component(LocalPose=Blend Poses by bool(BlendPose_0=@n12, BlendPose_1=@n10) {blendType=Linear})
    Alpha=1
    ExternalForce="0,0,0"
   ) {
    angularDampingOverride=0.999
    angularSpringConstant=10000
    bAngularSpring=true
    bOverrideAngularDamping=true
    bUseGravityOverride=true
    gravityScale=0.35
    simulationSpace=RootRelative
    alphaBoolBlend {blendOption=Linear}
    boundBone {boneName=dyn_frontSkirt_02_r}
    componentLinearAccScale=<1 1 1>
    componentLinearVelScale=<1 1 1>
    gravityOverride {z=-980}
    physicsBodyDefinitions [{
      boundBone {boneName=dyn_frontSkirt_02_r}
      boxExtents=<20 20 20>
      constraintSetup {angularLimitsMax=<0 5 0>, angularLimitsMin=<0 -40 0>, angularTarget=<1 0 0>}
      localJointOffset=<-4 0 0>
     }]
    relativeSpaceBone {boneName=spine_01}
   }
   n14 = Anim Dynamics - Bone: dyn_skirt_f_m(ComponentPose=@n13, Alpha=1, ExternalForce="0,0,0") {
    angularDampingOverride=0.999
    angularSpringConstant=1000
    bAngularSpring=true
    bOverrideAngularDamping=true
    bUseGravityOverride=true
    gravityScale=0.35
    simulationSpace=RootRelative
    alphaBoolBlend {blendOption=Linear}
    boundBone {boneName=dyn_skirt_f_m}
    componentLinearAccScale=<1 1 1>
    componentLinearVelScale=<1 1 1>
    gravityOverride {z=-980}
    physicsBodyDefinitions [{
      boundBone {boneName=dyn_skirt_f_m}
      boxExtents=<13 13 13>
      constraintSetup {angularLimitsMax=<0 10 0>, angularLimitsMin=<0 -20 0>, angularTarget=<1 0 0>}
      localJointOffset=<4 0 0>
     }]
    relativeSpaceBone {boneName=spine_01}
   }
   n15 = Anim Dynamics - Bone: dyn_frontSkirt_02_l(ComponentPose=@n14, Alpha=1, ExternalForce="0,0,0") {
    angularDampingOverride=0.999
    angularSpringConstant=10000
    bAngularSpring=true
    bOverrideAngularDamping=true
    bUseGravityOverride=true
    gravityScale=0.35
    simulationSpace=RootRelative
    alphaBoolBlend {blendOption=Linear}
    boundBone {boneName=dyn_frontSkirt_02_l}
    componentLinearAccScale=<1 1 1>
    componentLinearVelScale=<1 1 1>
    gravityOverride {z=-980}
    physicsBodyDefinitions [{
      boundBone {boneName=dyn_frontSkirt_02_l}
      boxExtents=<20 20 20>
      constraintSetup {angularLimitsMax=<0 5 0>, angularLimitsMin=<0 -40 0>, angularTarget=<1 0 0>}
      localJointOffset=<4 0 0>
     }]
    relativeSpaceBone {boneName=spine_01}
   }
   n16 = Anim Dynamics - Bone: dyn_backSkirt_02_r(ComponentPose=@n15, Alpha=1, ExternalForce="0,0,0") {
    angularDampingOverride=0.999
    angularSpringConstant=1000
    bAngularSpring=true
    bOverrideAngularDamping=true
    bUseGravityOverride=true
    gravityScale=0.35
    simulationSpace=RootRelative
    alphaBoolBlend {blendOption=Linear}
    boundBone {boneName=dyn_backSkirt_02_r}
    componentLinearAccScale=<1 1 1>
    componentLinearVelScale=<1 1 1>
    gravityOverride {z=-980}
    physicsBodyDefinitions [{
      boundBone {boneName=dyn_backSkirt_02_r}
      boxExtents=<15 15 15>
      constraintSetup {angularLimitsMax=<0 5 0>, angularLimitsMin=<0 -50 0>, angularTarget=<1 0 0>}
      localJointOffset=<-4 0 0>
     }]
    relativeSpaceBone {boneName=spine_01}
   }
   n17 = Anim Dynamics - Bone: dyn_skirt_b_m(ComponentPose=@n16, Alpha=1, ExternalForce="0,0,0") {
    angularDampingOverride=0.99
    angularSpringConstant=1000
    bAngularSpring=true
    bOverrideAngularDamping=true
    bUseGravityOverride=true
    gravityScale=0.35
    simulationSpace=RootRelative
    alphaBoolBlend {blendOption=Linear}
    boundBone {boneName=dyn_skirt_b_m}
    componentLinearAccScale=<1 1 1>
    componentLinearVelScale=<1 1 1>
    gravityOverride {z=-980}
    physicsBodyDefinitions [{
      boundBone {boneName=dyn_skirt_b_m}
      boxExtents=<13 13 13>
      constraintSetup {angularLimitsMax=<0 10 0>, angularLimitsMin=<0 -60 0>, angularTarget=<1 0 0>}
      localJointOffset=<4 0 0>
     }]
    relativeSpaceBone {boneName=spine_01}
   }
   Save cached pose 'animDynamics Skirt'(Pose=Component To Local(ComponentPose=@n17)) {cacheName=animDynamics Skirt}
   n18 = Control Rig!(
    Source=Copy Pose From Mesh() {bUseAttachedParent=true}
    Alpha
    "Crouch Moving ":bool=Get Crouch Walking() -> (Crouch Walking:bool)
    Moving Sides:bool=Get Moving side to side() -> (Moving side to side:bool)
    Standing:double="Get Standing "() -> ("Standing ":double)
    Pickaxe swing:bool=Get bIsPlayingMeleeAnim()
    is Front End Idle:bool=Get Front End Idle() -> (Front End Idle:bool)
   ) -> (Pose) {
    alphaBoolBlend {blendOption=Linear}
    controlRigAssetReference {blueprintRigClass=&/F_MED_TangyRadishMagma_CtrlRig.F_MED_TangyRadishMagma_CtrlRig_C}
   }
   n19 = "Transform (Modify) Bone - Bone: base_backSkirt_02_l"(ComponentPose=Local To Component(LocalPose=@n18), Alpha=1, Translation="0,0,1", Rotation="0,0,0", Scale="1,1,1") {
    rotationMode=BMM_Additive
    rotationSpace=BCS_BoneSpace
    translationMode=BMM_Additive
    translationSpace=BCS_BoneSpace
    alphaBoolBlend {blendOption=Linear}
    alphaScaleBiasClamp {bInterpResult=true}
    boneToModify {boneName=base_backSkirt_02_l}
    translation {z=1}
   }
   n20 = "Transform (Modify) Bone - Bone: thigh_l_mover"(ComponentPose=@n19, Alpha=1, Translation="0,-2,0", Rotation="0,0,0", Scale="1,1,1") {
    rotationMode=BMM_Additive
    rotationSpace=BCS_BoneSpace
    translationMode=BMM_Additive
    translationSpace=BCS_BoneSpace
    alphaBoolBlend {blendOption=Linear}
    alphaScaleBiasClamp {bInterpResult=true}
    boneToModify {boneName=thigh_l_mover}
    translation {y=-2}
   }
   n21 = "Transform (Modify) Bone - Bone: base_belt_02"(ComponentPose=@n20, Alpha=1, Translation="0,0,0", Rotation="0,0,-20", Scale="1,1,1") {
    rotationMode=BMM_Additive
    rotationSpace=BCS_BoneSpace
    translationMode=BMM_Additive
    translationSpace=BCS_BoneSpace
    alphaBoolBlend {blendOption=Linear}
    boneToModify {boneName=base_belt_02}
    rotation {roll=-20}
   }
   n22 = "Transform (Modify) Bone - Bone: base_belt_02"(ComponentPose=@n21, Alpha=1, Translation="0,0,0", Rotation="0,0,-25", Scale="1,1,1") {
    rotationMode=BMM_Additive
    rotationSpace=BCS_BoneSpace
    scaleMode=BMM_Additive
    scaleSpace=BCS_BoneSpace
    translationMode=BMM_Additive
    translationSpace=BCS_BoneSpace
    alphaBoolBlend {blendOption=Linear}
    boneToModify {boneName=base_belt_02}
    rotation {roll=-25}
   }
   Save cached pose 'TransformCache'(Pose=Component To Local(ComponentPose=@n22)) {cacheName=TransformCache}
  ]}
 EventGraph {nodes [
   Event BlueprintUpdateAnimation()
   n1 = Get bIsMoving2D()
   n2 = NOT Boolean(A=@n1)
   n3 = Get bIsCrouching()
   > Set Crouch Idle CR(Crouch Idle CR:double="To Float (Boolean)"(InBool=AND Boolean(A=@n3, B=@n2))) -> (Output_Get:double)
   > Set Crouch Walking(Crouch Walking:bool=AND Boolean(A=@n3, B=@n1)) -> (Output_Get:bool)
   > "Set Standing "("Standing ":double="To Float (Boolean)"(InBool=NOT Boolean(A=@n3))) -> (Output_Get:double)
   > n4 = PlayerMovement(AnimInstance=Self-Reference(), XDotProductCoefficient=100, YDotProductCoefficient=100, ZDotProductCoefficient=100)
   n5 = BreakVector(InVec=MakeVector(X=@n4, Y=@n4.YMovement, Z=@n4.ZMovement))
   > n6 = Set Left(Left:bool="float < float"(A=@n5, B=-50)) -> (Output_Get:bool)
   > n7 = Set Right(Right:bool="float > float"(A=@n5, B=50)) -> (Output_Get:bool)
   > Set Forward(Forward:bool="float > float"(A=@n5.Y, B=40)) -> (Output_Get:bool)
   > Set Backward(Backward:bool="float < float"(A=@n5.Y, B=-40)) -> (Output_Get:bool)
   > Set Moving side to side(Moving side to side:bool=OR Boolean(A=@n6, B=@n7)) -> (Output_Get:bool)
   n8 = Get bIsPlayingEmote()
   > Set Front End Idle(Front End Idle:bool=AND Boolean(A=Get bIsInFrontEnd(), B=NOT Boolean(A=@n8))) -> (Output_Get:bool)
   > n9 = Branch(Condition=Get bIsInFrontEnd())
   > Set Custom Idle is Playing Float(Custom Idle is Playing Float:double=GetCurveValue(self=GetOwnerFrontEndAnimInstance(), CurveName=Frontend_Idle_Bypass)) -> (Output_Get:double)
   Set Custom Idle is Playing Float(execute=@n9.else, Custom Idle is Playing Float=0)
  ]}
]
variables [
 Backward {type=boolean}
 Crouch Idle CR {type=number}
 Crouch Walking {type=boolean}
 Custom Idle is Playing Float {type=number}
 Forward {type=boolean}
 Front End Idle {type=boolean}
 Left {type=boolean}
 Moving side to side {type=boolean}
 Right {type=boolean}
 "Standing " {type=number}
]
