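# Post-copy AnimBP for the FaceAcc part of the TangyRadishMagma female character, adding secondary hood motion.
# Copies the attached parent's pose, runs a Control Rig, then chains spring-based Anim Dynamics on side, top and bottom hood bones.
# The cached result is used when backbling or cape is worn; otherwise a further back-hood simulation is added.
# Customization load sets those flags, plus front-end-dependent hood space and alpha variables that the AnimGraph never reads.
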AnimBlueprint /BRCosmetics/Characters/Player/Female/Medium/Bodies/F_MED_TangyRadishMagma/Meshes/Parts/F_MED_TangyRadishMagma_FaceAcc_AnimBP
target_skeleton=&/F_MED_TangyRadishMagma_FaceAcc_Skeleton
graphs [
 AnimGraph {nodes [
   Output Pose(Result=Blend Poses by bool(
     BlendPose_0=Use cached pose 'Hood Up Dynamics'()
     BlendPose_1=Use cached pose 'Hood Down Dynamics'()
     bActiveValue=Get is Wearing Backbling or Cape() -> (is Wearing Backbling or Cape:bool)
    ) {blendType=Linear}) {layerGroup=None}
   n1 = Control Rig(Source=Copy Pose From Mesh() {bUseAttachedParent=true}) {
    alphaBoolBlend {blendOption=Linear}
    controlRigAssetReference {blueprintRigClass=&/F_MED_TangyRadishMagma_FaceAcc_CtrlRig.F_MED_TangyRadishMagma_FaceAcc_CtrlRig_C}
   }
   n2 = Anim Dynamics - Bone: dyn_l_hood!(SimulationSpace:AnimPhysSimSpaceType Enum, ComponentPose=Local To Component(LocalPose=@n1), Alpha=1, ExternalForce="0,0,0") -> (Pose) {
    angularDampingOverride=1
    angularSpringConstant=5000
    bAngularSpring=true
    bOverrideAngularDamping=true
    bUseGravityOverride=true
    gravityScale=0.35
    simulationSpace=RootRelative
    alphaBoolBlend {blendOption=Linear}
    boundBone {boneName=dyn_l_hood}
    componentAppliedLinearAccClamp=<3000 3000 3000>
    componentLinearAccScale=<1 1 1>
    componentLinearVelScale=<1 1 1>
    gravityOverride {z=-980}
    physicsBodyDefinitions [{
      boundBone {boneName=dyn_l_hood}
      boxExtents=<16 16 16>
      constraintSetup {angularLimitsMax=<5 0 25>, angularLimitsMin=<-5 0 -10>, angularTarget=<1 0 0>}
      localJointOffset=<2 0 0>
     }]
    relativeSpaceBone {boneName=neck_01}
   }
   n3 = Anim Dynamics - Bone: dyn_r_hood!(SimulationSpace:AnimPhysSimSpaceType Enum=BoneRelative, ComponentPose=@n2, Alpha=1, ExternalForce="0,0,0") -> (Pose) {
    angularDampingOverride=1
    angularSpringConstant=5000
    bAngularSpring=true
    bOverrideAngularDamping=true
    bUseGravityOverride=true
    gravityScale=0.35
    simulationSpace=BoneRelative
    alphaBoolBlend {blendOption=Linear}
    boundBone {boneName=dyn_r_hood}
    componentLinearAccScale=<1 1 1>
    componentLinearVelScale=<1 1 1>
    gravityOverride {z=-980}
    physicsBodyDefinitions [{
      boundBone {boneName=dyn_r_hood}
      boxExtents=<13 13 13>
      constraintSetup {angularLimitsMax=<5 0 25>, angularLimitsMin=<-5 0 -10>, angularTarget=<1 0 0>}
      localJointOffset=<-2 0 0>
     }]
    relativeSpaceBone {boneName=neck_01}
   }
   n4 = Anim Dynamics - Bone: dyn_top_hood!(SimulationSpace:AnimPhysSimSpaceType Enum, ComponentPose=@n3, Alpha=1, ExternalForce="0,0,0") -> (Pose) {
    angularDampingOverride=1
    angularSpringConstant=5000
    bAngularSpring=true
    bOverrideAngularDamping=true
    bUseGravityOverride=true
    gravityScale=0.35
    simulationSpace=RootRelative
    alphaBoolBlend {blendOption=Linear}
    boundBone {boneName=dyn_top_hood}
    componentAppliedLinearAccClamp=<3000 3000 3000>
    componentLinearAccScale=<1 1 1>
    componentLinearVelScale=<1 1 1>
    gravityOverride {z=-980}
    physicsBodyDefinitions [{
      boundBone {boneName=dyn_top_hood}
      boxExtents=<20 20 20>
      constraintSetup {angularLimitsMax=<0 2 2>, angularLimitsMin=<0 -30 -2>, angularTarget=<1 0 0>}
      localJointOffset=<2 0 0>
     }]
    relativeSpaceBone {boneName=neck_01}
   }
   n5 = Anim Dynamics - Bone: dyn_l_bott_hood(ComponentPose=@n4, Alpha=1, ExternalForce="0,0,0") {
    angularDampingOverride=1
    angularSpringConstant=5000
    bAngularSpring=true
    bOverrideAngularDamping=true
    bUseGravityOverride=true
    gravityScale=0.35
    simulationSpace=BoneRelative
    alphaBoolBlend {blendOption=Linear}
    boundBone {boneName=dyn_l_bott_hood}
    componentAppliedLinearAccClamp=<3000 3000 3000>
    componentLinearAccScale=<1 1 1>
    componentLinearVelScale=<1 1 1>
    gravityOverride {z=-980}
    physicsBodyDefinitions [{
      boundBone {boneName=dyn_l_bott_hood}
      boxExtents=<10 10 10>
      constraintSetup {angularLimitsMax=<15 15 15>, angularLimitsMin=<-15 -15 -15>, angularTarget=<1 0 0>}
      localJointOffset=<2 0 0>
     }]
    relativeSpaceBone {boneName=root}
   }
   n6 = Anim Dynamics - Bone: dyn_r_bott_hood(ComponentPose=@n5, Alpha=1, ExternalForce="0,0,0") {
    angularDampingOverride=1
    angularSpringConstant=5000
    bAngularSpring=true
    bOverrideAngularDamping=true
    bUseGravityOverride=true
    gravityScale=0.35
    simulationSpace=BoneRelative
    alphaBoolBlend {blendOption=Linear}
    boundBone {boneName=dyn_r_bott_hood}
    componentAppliedLinearAccClamp=<3000 3000 3000>
    componentLinearAccScale=<1 1 1>
    componentLinearVelScale=<1 1 1>
    gravityOverride {z=-980}
    physicsBodyDefinitions [{
      boundBone {boneName=dyn_r_bott_hood}
      boxExtents=<10 10 10>
      constraintSetup {angularLimitsMax=<15 15 15>, angularLimitsMin=<-15 -15 -15>, angularTarget=<1 0 0>}
      localJointOffset=<2 0 0>
     }]
    relativeSpaceBone {boneName=root}
   }
   Save cached pose 'Hood Up Dynamics'(Pose=Component To Local(ComponentPose=@n6)) {cacheName=Hood Up Dynamics}
   n7 = Anim Dynamics - Bone: dyn_back_hood(ComponentPose=Local To Component(LocalPose=Use cached pose 'Hood Up Dynamics'()), Alpha=1, ExternalForce="0,0,0") {
    angularDampingOverride=1
    angularSpringConstant=1000
    bAngularSpring=true
    bOverrideAngularDamping=true
    bUseGravityOverride=true
    gravityScale=0.35
    simulationSpace=BoneRelative
    alphaBoolBlend {blendOption=Linear}
    boundBone {boneName=dyn_back_hood}
    componentAppliedLinearAccClamp=<3000 3000 3000>
    componentLinearAccScale=<1 1 1>
    componentLinearVelScale=<1 1 1>
    gravityOverride {z=-980}
    physicsBodyDefinitions [{
      boundBone {boneName=dyn_back_hood}
      boxExtents=<25 25 25>
      constraintSetup {angularLimitsMax=<0 1 25>, angularLimitsMin=<0 -70 -25>, angularTarget=<1 0 0>}
      localJointOffset=<2 0 0>
     }]
    relativeSpaceBone {boneName=root}
   }
   Save cached pose 'Hood Down Dynamics'(Pose=Component To Local(ComponentPose=@n7)) {cacheName=Hood Down Dynamics}
  ]}
 EventGraph {nodes [
   n1 = Event OnCharacterCustomizationLoaded()
   > Set Skel_Mesh(Skel_Mesh=@n1.CustomBaseMesh)
   n2 = CheckForBackbling()
   > Set is Wearing Backbling(is Wearing Backbling:bool=AND Boolean!(A=@n2, B=NOT Boolean(A=@n2.bIsBackblingACape), C:bool=NOT Boolean(A=@n2.bIsBackblingLong)) -> (ReturnValue)) -> (Output_Get:bool)
   > Set is Wearing Cape(is Wearing Cape:bool=OR Boolean(A=@n2.bIsBackblingACape, B=@n2.bIsBackblingLong)) -> (Output_Get:bool)
   n3 = OR Boolean(A=Get is Wearing Backbling() -> (is Wearing Backbling:bool), B=Get is Wearing Cape() -> (is Wearing Cape:bool))
   > Set is Wearing Backbling or Cape(is Wearing Backbling or Cape:bool=@n3) -> (Output_Get:bool)
   > Set is Wearing Backbling or Cape Alpha(is Wearing Backbling or Cape Alpha:double="To Float (Boolean)"(InBool=@n3)) -> (Output_Get:double)
   > n4 = Branch(Condition=Get bIsInFrontEnd())
   > Set Hood Sim Space(Hood Sim Space:AnimPhysSimSpaceType Enum=World) -> (Output_Get:AnimPhysSimSpaceType Enum)
   > Set Hood Alpha(Hood Alpha:float=0.6) -> (Output_Get:float)
   Set Hood Sim Space(execute=@n4.else)
   > Set Hood Alpha(Hood Alpha=1)
   Event BlueprintUpdateAnimation()
  ]}
]
variables [
 Hood Alpha {type=number}
 Hood Sim Space {type=AnimPhysSimSpaceType}
 is Wearing Backbling {type=boolean}
 is Wearing Backbling or Cape {type=boolean}
 is Wearing Backbling or Cape Alpha {type=number}
 is Wearing Cape {type=boolean}
 Skel_Mesh {type=SkeletalMeshComponent}
]
